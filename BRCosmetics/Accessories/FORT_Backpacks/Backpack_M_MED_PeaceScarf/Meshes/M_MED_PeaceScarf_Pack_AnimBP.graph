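# Animation Blueprint for the PeaceScarf backpack that makes its balloon and string sway as the player moves.
# BalloonUp blends in a string-tip offset during skydiving, skydive floating or DBNO, unless the parachute is open.
# After the bone offsets, chained spring AnimDynamics simulate the balloon with upward gravity and three string bones with downward gravity.
# A sine-based BalloonFloat bob is computed when not falling, but the AnimGraph does not read it.
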AnimBlueprint /BRCosmetics/Accessories/FORT_Backpacks/Backpack_M_MED_PeaceScarf/Meshes/M_MED_PeaceScarf_Pack_AnimBP
target_skeleton=&/M_MED_PeaceScarf_Pack_Skeleton
graphs [
 AnimGraph {nodes [
   n1 = Input Pose()
   n2 = "Transform (Modify) Bone - Bone: base_balloon"(
    ComponentPose="Transform (Modify) Bone - Bone: base_balloon"(ComponentPose=Local To Component(LocalPose=@n1), Alpha=1, Translation="0,0,0", Rotation="0,0,0", Scale="1,1,1") {translationMode=BMM_Additive, translationSpace=BCS_BoneSpace, boneToModify {boneName=base_balloon}}
    Alpha=1
    Translation="0,0,0"
    Rotation="0,0,0"
    Scale="1,1,1"
   ) {
    rotationMode=BMM_Additive
    rotationSpace=BCS_BoneSpace
    translationMode=BMM_Additive
    translationSpace=BCS_BoneSpace
    boneToModify {boneName=base_balloon}
   }
   n3 = "Transform (Modify) Bone - Bone: base_string_tip"!(ComponentPose=@n2, bAlphaBoolEnabled:bool=Get BalloonUp() -> (BalloonUp:bool), Translation="0,10,32", Rotation="0,0,-60", Scale="1,1,1") -> (Pose) {
    alphaInputType=Bool
    rotationMode=BMM_Additive
    rotationSpace=BCS_BoneSpace
    translationMode=BMM_Additive
    translationSpace=BCS_BoneSpace
    alphaBoolBlend {blendInTime=1, blendOutTime=2}
    alphaScaleBiasClamp {bInterpResult=true, interpSpeedIncreasing=1}
    boneToModify {boneName=base_string_tip}
    rotation {roll=-60}
    translation {y=10, z=32}
   }
   n4 = "Transform (Modify) Bone - Bone: dyn_balloon_string_01"!(ComponentPose=@n3, bAlphaBoolEnabled:bool=True, Translation="0,0,0", Rotation="-35,0,0", Scale="1,1,1") -> (Pose) {
    alphaInputType=Bool
    rotationMode=BMM_Additive
    rotationSpace=BCS_BoneSpace
    translationMode=BMM_Additive
    translationSpace=BCS_BoneSpace
    alphaBoolBlend {blendInTime=3, blendOutTime=3}
    alphaScaleBiasClamp {bInterpResult=true, interpSpeedIncreasing=1}
    boneToModify {boneName=dyn_balloon_string_01}
    rotation {pitch=-35}
   }
   n5 = Anim Dynamics - Bone: dyn_balloon(ComponentPose=@n4, Alpha=1, ExternalForce="0,0,0") {
    angularDampingOverride=1
    angularSpringConstant=3000
    bAngularSpring=true
    bOverrideAngularDamping=true
    bUseGravityOverride=true
    gravityScale=0.35
    simulationSpace=RootRelative
    boundBone {boneName=dyn_balloon}
    componentAppliedLinearAccClamp=<3000 3000 3000>
    componentLinearAccScale=<0.3 0.3 0.3>
    componentLinearVelScale=<0.3 0.3 0.3>
    gravityOverride {z=200}
    physicsBodyDefinitions [{
      boundBone {boneName=dyn_balloon}
      boxExtents=<30 30 30>
      constraintSetup {angularLimitsMax=<30 25 45>, angularLimitsMin=<-30 -45 -45>, angularTarget=<1 0 0>}
      localJointOffset=<-4 0 0>
     }]
    simSpaceSettings {maxAngularAcceleration=100, maxAngularVelocity=10, simSpaceAngularAlpha=1}
   }
   n6 = Anim Dynamics - Bone: dyn_balloon_string_01!(
    ComponentPose=@n5
    Alpha=1
    ComponentLinearAccScale:Vector="0.5,0.5,0.5"
    ComponentLinearVelScale:Vector="0.5,0.5,0.5"
    ExternalForce="0,0,0"
   ) -> (Pose) {
    angularDampingOverride=1
    angularSpringConstant=6000
    bAngularSpring=true
    bOverrideAngularDamping=true
    bUseGravityOverride=true
    gravityScale=0.35
    simulationSpace=RootRelative
    boundBone {boneName=dyn_balloon_string_01}
    componentAppliedLinearAccClamp=<3000 3000 3000>
    componentLinearAccScale=<0.5 0.5 0.5>
    componentLinearVelScale=<0.5 0.5 0.5>
    gravityOverride {z=-980}
    physicsBodyDefinitions [{
      boundBone {boneName=dyn_balloon_string_01}
      boxExtents=<25 25 25>
      constraintSetup {angularLimitsMax=<15 30 55>, angularLimitsMin=<-15 -15 -55>, angularTarget=<1 0 0>}
      localJointOffset=<6 0 0>
     }]
    simSpaceSettings {maxAngularAcceleration=100, maxAngularVelocity=10, simSpaceAngularAlpha=1}
   }
   n7 = Anim Dynamics - Bone: dyn_balloon_string_02!(
    ComponentPose=@n6
    Alpha=1
    ComponentLinearAccScale:Vector="0.5,0.5,0.5"
    ComponentLinearVelScale:Vector="0.5,0.5,0.5"
    ExternalForce="0,0,0"
   ) -> (Pose) {
    angularDampingOverride=1
    angularSpringConstant=3000
    bAngularSpring=true
    bOverrideAngularDamping=true
    bUseGravityOverride=true
    gravityScale=0.35
    simulationSpace=RootRelative
    boundBone {boneName=dyn_balloon_string_02}
    componentAppliedLinearAccClamp=<3000 3000 3000>
    componentLinearAccScale=<0.5 0.5 0.5>
    componentLinearVelScale=<0.5 0.5 0.5>
    gravityOverride {z=-980}
    physicsBodyDefinitions [{
      boundBone {boneName=dyn_balloon_string_02}
      boxExtents=<25 25 25>
      constraintSetup {angularLimitsMax=<15 30 55>, angularLimitsMin=<-15 -15 -55>, angularTarget=<1 0 0>}
      localJointOffset=<6 0 0>
     }]
    simSpaceSettings {maxAngularAcceleration=100, maxAngularVelocity=10, simSpaceAngularAlpha=1}
   }
   n8 = Anim Dynamics - Bone: dyn_balloon_string_03!(
    ComponentPose=@n7
    Alpha=1
    ComponentLinearAccScale:Vector="0.5,0.5,0.5"
    ComponentLinearVelScale:Vector="0.5,0.5,0.5"
    ExternalForce="0,0,0"
   ) -> (Pose) {
    angularDampingOverride=1
    angularSpringConstant=3000
    bAngularSpring=true
    bOverrideAngularDamping=true
    bUseGravityOverride=true
    gravityScale=0.35
    simulationSpace=RootRelative
    boundBone {boneName=dyn_balloon_string_03}
    componentAppliedLinearAccClamp=<3000 3000 3000>
    componentLinearAccScale=<0.5 0.5 0.5>
    componentLinearVelScale=<0.5 0.5 0.5>
    gravityOverride {z=-980}
    physicsBodyDefinitions [{
      boundBone {boneName=dyn_balloon_string_03}
      boxExtents=<25 25 25>
      constraintSetup {angularLimitsMax=<15 30 55>, angularLimitsMin=<-15 -15 -55>, angularTarget=<1 0 0>}
      localJointOffset=<6 0 0>
     }]
    simSpaceSettings {maxAngularAcceleration=100, maxAngularVelocity=10, simSpaceAngularAlpha=1}
   }
   Output Pose(Result=Component To Local(ComponentPose=@n8)) {layerGroup=None}
   Get ZiplineUse() -> (ZiplineUse:Vector)
   Get Is Falling not Skydive() -> (Is Falling not Skydive:float)
  ]}
 EventGraph {nodes [
   Event BlueprintUpdateAnimation()
   > n1 = Sequence!(execute) -> (then_0, then_1, then_2:Exec)
   > n2 = Branch(Condition=Get bIsFalling())
   > Set BalloonFloat(BalloonFloat:Vector="0, 0, 0") -> (Output_Get:Vector)
   n3 = GetRealTimeSeconds()
   n4 = "Sin (Radians)"(A=float * float(A=float * float(A=1.2, B=1.3), B=@n3))
   n5 = "Sin (Radians)"(A=float * float(A=float * float(A=0.9, B=1.1), B=@n3))
   n6 = "Sin (Radians)"(A=float * float(A=float * float(A=1.1, B=1.1), B=@n3))
   Set BalloonFloat(
    execute=@n2.else
    BalloonFloat=MakeVector(
     X=Remap Float(Input=@n4, OldRange=(X=-1,Y=1), NewRange=(X=-2,Y=0)) -> (Output=~)
     Y=float * float(A=@n6, B=0.15)
     Z=float * float(A=@n5, B=0.3)
    )
   )
   n7 = Get bIsParachuteOpened()
   n8 = Get bIsSkyDiving()
   n9 = Get bIsSkydiveFloating()
   n10 = Get bIsDBNO()
   n11 = Branch(execute=@n1.then_2, Condition=AND Boolean(A=NOT Boolean(A=@n7), B=OR Boolean!(A=@n7, B=@n8, C:bool=@n9, D:bool=@n10) -> (ReturnValue)))
   > Set BalloonUp(BalloonUp:bool=true) -> (Output_Get:bool)
   Set BalloonUp(execute=@n11.else, BalloonUp=false)
   Get bIsFalling()
   n12 = NOT Boolean(A=Get bIsSkyDiving())
   n13 = Get bIsSkydiveFloating()
   Set Is Falling not Skydive(Is Falling not Skydive:float="To Float (Boolean)"(InBool=AND Boolean(A=@n12, B=@n13))) -> (Output_Get:float)
  ]}
]
variables [
 animDyn on zipline {type=Rotator}
 BalloonFloat {type=Vector}
 BalloonUp {type=boolean}
 Is Falling not Skydive {type=number, default=1}
 ZiplineUse {type=Vector}
]
